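# Sound cue for the Hot Hands applied state in Fire Petal Boons, mixing a random one-shot sting under a looping bed.
# An IsLocal branch gives the local player a random loop with enveloped volume swell, decay and falling pitch.
# Otherwise, including when unset, a random band-passed loop starts after a delay, quieter and slowly fading in.
# Concurrency allows one instance per owner, blocking new ones, with third-person attenuation.
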SoundCue /FirePetalBoonsGameplay/Assets/Audio/Cues/SC_HotHandsState_Applied_Loop
root=Mixer_0
nodes [
 {id=Branch_0, children [Enveloper_0,Delay_0,Delay_0], properties {boolParameterName=IsLocal}}
 {id=Delay_0, children [Modulator_0], properties {delayMax=2, delayMin=2}}
 {
  id=Enveloper_0
  children [Random_0]
  properties {
   pitchCurve {editorCurveData {keys [<at 0 1.25>, {time=6}]}}
   volumeCurve {editorCurveData {keys [<at 0 0>, {arriveTangent=0.0333, interpMode=RCIM_Cubic, leaveTangent=0.0333, time=0.5, value=1}, <at 6 0.2>]}}
  }
 }
 {
  id=Enveloper_1
  children [Random_1]
  properties {pitchCurve {editorCurveData {keys [{time=0}]}}, volumeCurve {editorCurveData {keys [<at 0 0>,<at 3.1956 0.6375>,<at 7 1>]}}}
 }
 {id=Mixer_0, children [Modulator_1,Branch_0], properties {inputVolume [0.2,1]}}
 {id=Modulator_0, children [Enveloper_1], properties {volumeMax=0.25, volumeMin=0.2}}
 {id=Modulator_1, children [Random_2]}
 {id=Random_0, children [WavePlayer_0,WavePlayer_1,WavePlayer_2], properties {bShouldExcludeFromBranchCulling=true, weights [1,1,1]}}
 {id=Random_1, children [WavePlayer_3,WavePlayer_4,WavePlayer_5], properties {bShouldExcludeFromBranchCulling=true, weights [1,1,1]}}
 {
  id=Random_2
  children [WavePlayer_6, WavePlayer_7, WavePlayer_8, WavePlayer_9, WavePlayer_10, WavePlayer_11]
  properties {bShouldExcludeFromBranchCulling=true, weights [1,1,1,1,1,1]}
 }
 {id=WavePlayer_0, properties {bLooping=true, soundWaveAssetPtr=/FirePetalBoonsGameplay/Assets/Audio/Waves/HotHandsState_Applied_Loop_03}}
 {id=WavePlayer_1, properties {bLooping=true, soundWaveAssetPtr=/FirePetalBoonsGameplay/Assets/Audio/Waves/HotHandsState_Applied_Loop_02}}
 {id=WavePlayer_10, properties {soundWaveAssetPtr=/FirePetalBoonsGameplay/Assets/Audio/Waves/HotHandsState_Applied_05}}
 {id=WavePlayer_11, properties {soundWaveAssetPtr=/FirePetalBoonsGameplay/Assets/Audio/Waves/HotHandsState_Applied_06}}
 {id=WavePlayer_2, properties {bLooping=true, soundWaveAssetPtr=/FirePetalBoonsGameplay/Assets/Audio/Waves/HotHandsState_Applied_Loop_01}}
 {
  id=WavePlayer_3
  properties {bLooping=true, soundWaveAssetPtr=/FirePetalBoonsGameplay/Assets/Audio/Waves/HotHandsState_Applied_BandPassed_Loop_01}
 }
 {
  id=WavePlayer_4
  properties {bLooping=true, soundWaveAssetPtr=/FirePetalBoonsGameplay/Assets/Audio/Waves/HotHandsState_Applied_BandPassed_Loop_02}
 }
 {
  id=WavePlayer_5
  properties {bLooping=true, soundWaveAssetPtr=/FirePetalBoonsGameplay/Assets/Audio/Waves/HotHandsState_Applied_BandPassed_Loop_03}
 }
 {id=WavePlayer_6, properties {soundWaveAssetPtr=/FirePetalBoonsGameplay/Assets/Audio/Waves/HotHandsState_Applied_01}}
 {id=WavePlayer_7, properties {soundWaveAssetPtr=/FirePetalBoonsGameplay/Assets/Audio/Waves/HotHandsState_Applied_02}}
 {id=WavePlayer_8, properties {soundWaveAssetPtr=/FirePetalBoonsGameplay/Assets/Audio/Waves/HotHandsState_Applied_03}}
 {id=WavePlayer_9, properties {soundWaveAssetPtr=/FirePetalBoonsGameplay/Assets/Audio/Waves/HotHandsState_Applied_04}}
]
properties {
 attenuationSettings=/FirePetalBoonsGameplay/Assets/Audio/Settings/Att_HotHandsState_Player_Bashable_Loop_3P
 bOverrideConcurrency=true
 maxDistance=3500
 soundClassObject=&Fort_Gadgets
 volumeMultiplier=0.3
 concurrencyOverrides {bLimitToOwner=true, maxCount=1, resolutionRule=PreventNew}
}
